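# Root conversation database for the CubicVice NPC, entered through the Conversation.Entry.CubicVice tag.
# A single speech node picks a greeting from contextual lines gated by first meeting, per-match conversation count, or the player's outfit, with priority curves on the first-meeting and outfit lines.
# It then fans out through link nodes to remote entries for quests, basic services, CubicVice dialogue, and Season 33 and 36 story content.
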ConversationDatabase /NPCLibrary/NPCs/CubicVice/NPCLibrary_Conversation_CubicVice
entry_points [{tag=Conversation.Entry.CubicVice, destinations [#1]}]
nodes [
 ConversationEntryPointNode_7 {
  id=#1
  class=ConversationEntryPointNode
  outputs [#2]
  properties {entryTag {tagName=Conversation.Entry.CubicVice}}
 }
 Speech_C_4 {
  id=#2
  class=Speech_C
  outputs [
   #3
   #4
   #5
   #6
   #7
   #8
   #9
   #10
   #11
   #12
   #13
   #14
  ]
  properties {
   nodeName=""
   generalConfig {
    defaultMessage=I've never felt more lost.
    contextualMessages [
     {
      message="My blade's at your service, if you need it."
      contextRequirements [{requirement {class=FortControllerRequirement_IsFirstEverConversationWithNPC}}]
      priority {curve {rowName=Tandem.Speech.Priority.FirstEverInteraction}}
     }
     {
      message=I've never felt more lost.
      contextRequirements [{requirement {class=FortControllerRequirement_CheckNPCMatchConversationCount, properties {nPCUniqueIdTag {tagName=AISpawnerData.Type.Tandem.CubicVice}}}}]
      priority {curve {rowName=None}}
     }
     {
      message=My brother is a traitor. What do I do now?
      contextRequirements [{requirement {class=FortControllerRequirement_CheckNPCMatchConversationCount, properties {nPCUniqueIdTag {tagName=AISpawnerData.Type.Tandem.CubicVice}}}}]
      priority {curve {rowName=None}}
     }
     {
      message="Every time I catch my breath, a new crisis hits."
      contextRequirements [{requirement {class=FortControllerRequirement_CheckNPCMatchConversationCount, properties {nPCUniqueIdTag {tagName=AISpawnerData.Type.Tandem.CubicVice}}}}]
      priority {curve {rowName=None}}
     }
     {
      message=This world moves so fast. I can't keep up.
      contextRequirements [{requirement {class=FortControllerRequirement_CheckNPCMatchConversationCount, properties {nPCUniqueIdTag {tagName=AISpawnerData.Type.Tandem.CubicVice}}}}]
      priority {curve {rowName=None}}
     }
     {
      message=I'll never forgive you for your betrayal. Never.
      contextRequirements [{requirement {class=FortControllerRequirement_HasCID, properties {allowedCIds [&Character_MistMylar_Claw]}}}]
      priority {curve {rowName=Tandem.Speech.Priority.OutfitReaction}}
     }
     {
      message="Don't be foolish, Jade. Daigo has turned his back on us."
      contextRequirements [{requirement {
         class=FortControllerRequirement_HasCID
         properties {allowedCIds [/BRCosmetics/Athena/Items/Cosmetics/Characters/Character_PolishedJade_Mind]}
        }}]
      priority {curve {rowName=Tandem.Speech.Priority.OutfitReaction}}
     }
     {
      message="What do you see, when you look at me? A hero, or a failure?"
      contextRequirements [{requirement {class=FortControllerRequirement_HasCID, properties {allowedCIds [&Character_IronLilac_Sly]}}}]
      priority {curve {rowName=Tandem.Speech.Priority.OutfitReaction}}
     }
     {
      message="If Hope sees potential in you, you should listen."
      contextRequirements [{requirement {class=FortControllerRequirement_HasCID, properties {allowedCIds [/BRCosmetics/Athena/Items/Cosmetics/Characters/Character_MistRavenPeck]}}}]
      priority {curve {rowName=Tandem.Speech.Priority.OutfitReaction}}
     }
     {
      message="Do you think my brother... No, there's no way."
      contextRequirements [{requirement {class=FortControllerRequirement_HasCID, properties {allowedCIds [&Character_RoseDepth_Seed]}}}]
      priority {curve {rowName=Tandem.Speech.Priority.OutfitReaction}}
     }
     {
      message=If my brother truly has betrayed us... do what you must.
      contextRequirements [{requirement {
         class=FortControllerRequirement_HasCID
         properties {allowedCIds [&Character_UndergroundRebel_Fashion]}
        }}]
      priority {curve {rowName=Tandem.Speech.Priority.OutfitReaction}}
     }
     {
      message="What is it you want? Money, power, an even fancier coat?"
      contextRequirements [{requirement {class=FortControllerRequirement_HasCID, properties {allowedCIds [/BRCosmetics/Athena/Items/Cosmetics/Characters/Character_MoralKoreLine]}}}]
      priority {curve {rowName=Tandem.Speech.Priority.OutfitReaction}}
     }
    ]
   }
  }
 }
 ConversationLinkNode_11 {
  id=#14
  class=ConversationLinkNode
  properties {remoteEntryTag {tagName=Conversation.Entry.Season36.RageDebris.CharacterStories}}
 }
 ConversationLinkNode_10 {
  id=#13
  class=ConversationLinkNode
  properties {remoteEntryTag {tagName=Conversation.Entry.Season36.CubicVice.CharacterStories.05}}
 }
 ConversationLinkNode_6 {
  id=#12
  class=ConversationLinkNode
  properties {remoteEntryTag {tagName=Conversation.Entry.Season36.CubicVice.CharacterStories.01}}
 }
 ConversationLinkNode_5 {
  id=#11
  class=ConversationLinkNode
  properties {remoteEntryTag {tagName=Conversation.Entry.Season33.Weaponexpertise.Melee.Complete.01}}
 }
 ConversationLinkNode_4 {id=#10, class=ConversationLinkNode, properties {remoteEntryTag {tagName=Conversation.Entry.CubicVice.Dialogue}}}
 ConversationLinkNode_9 {
  id=#9
  class=ConversationLinkNode
  properties {remoteEntryTag {tagName=Conversation.Entry.Season33.Story.MM.06}}
 }
 ConversationLinkNode_8 {
  id=#8
  class=ConversationLinkNode
  properties {remoteEntryTag {tagName=Conversation.Entry.Season33.Story.MM.01}}
 }
 ConversationLinkNode_7 {
  id=#7
  class=ConversationLinkNode
  properties {remoteEntryTag {tagName=Conversation.Entry.Season33.Story.Confrontation.01}}
 }
 ConversationLinkNode_2 {
  id=#6
  class=ConversationLinkNode
  properties {remoteEntryTag {tagName=Conversation.Entry.Season33.Story.StopMM.01}}
 }
 ConversationLinkNode_3 {
  id=#5
  class=ConversationLinkNode
  properties {remoteEntryTag {tagName=Conversation.Entry.Season33.Story.Meet.01}}
 }
 ConversationLinkNode_0 {id=#4, class=ConversationLinkNode, properties {remoteEntryTag {tagName=Conversation.Entry.BasicServices}}}
 ConversationLinkNode_1 {id=#3, class=ConversationLinkNode, properties {remoteEntryTag {tagName=Conversation.Entry.NPCLibrary.Quests}}}
]
